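# Animation Blueprint for the FrankenLaze biped cosmetic companion that flags facial pose changes to its materials.
# Each update reads brow, eye, jaw and mouth curves, sampling only right-side poses for sided ones.
# Threshold crossings against previous-frame values, some requiring paired crossings, are ORed into a HasPoseChanged scalar on the owner's skeletal mesh materials.
# A Sequence then caches current values for the next frame, so the flag pulses only on crossing frames.
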AnimBlueprint /CosmeticCompanions/Assets/Biped/FrankenLaze/Blueprint/FrankenLaze_Base_AnimBP_Child
target_skeleton=/CosmeticCompanions/Assets/Biped/FrankenLaze/Meshes/FrankenLaze_Biped_Skeleton
graphs [
 AllowsCinematicOverride {nodes [
   AllowsCinematicOverride() {is=title K2Node_FunctionEntry}
   > Return Node!(execute, ReturnValue:bool=false)
  ]}
 EventGraph {nodes [
   Event BlueprintUpdateAnimation()
   > n1 = Sequence!(execute) -> (then_0, then_1, then_2:Exec)
   > Set R_brow_In_Dn_pose(R_brow_In_Dn_pose:double=GetCurveValue(CurveName=R_brow_In_Dn_pose)) -> (Output_Get:double)
   > Set R_brow_Ot_Dn_pose(R_brow_Ot_Dn_pose:double=GetCurveValue(CurveName=R_brow_Ot_Dn_pose)) -> (Output_Get:double)
   > Set R_brow_In_Up_pose(R_brow_In_Up_pose:double=GetCurveValue(CurveName=R_brow_In_Up_pose)) -> (Output_Get:double)
   > Set R_brow_Ot_Up_pose(R_brow_Ot_Up_pose:double=GetCurveValue(CurveName=R_brow_Ot_Up_pose)) -> (Output_Get:double)
   > Set R_brow_squeeze_in_pose(R_brow_squeeze_in_pose:double=GetCurveValue(CurveName=R_brow_squeeze_in_pose)) -> (Output_Get:double)
   > Set brow_emote_01(brow_emote_01:double=GetCurveValue(CurveName=brow_emote_01)) -> (Output_Get:double)
   > Set R_blink_pose(R_blink_pose:double=GetCurveValue(CurveName=R_blink_pose)) -> (Output_Get:double)
   > Set R_upperLidDown_pose(R_upperLidDown_pose:double=GetCurveValue(CurveName=R_upperLidDown_pose)) -> (Output_Get:double)
   > Set R_smile_pose(R_smile_pose:double=GetCurveValue(CurveName=R_smile_pose)) -> (Output_Get:double)
   > Set R_lower_lid_up_pose(R_lower_lid_up_pose:double=GetCurveValue(CurveName=R_lower_lid_up_pose)) -> (Output_Get:double)
   > Set eye_emote_01(eye_emote_01:double=GetCurveValue(CurveName=eye_emote_01)) -> (Output_Get:double)
   > Set R_frown_pose(R_frown_pose:double=GetCurveValue(CurveName=R_frown_pose)) -> (Output_Get:double)
   > Set jaw_open_pose(jaw_open_pose:double=GetCurveValue(CurveName=jaw_open_pose)) -> (Output_Get:double)
   > Set reverse_jaw_pose(reverse_jaw_pose:double=GetCurveValue(CurveName=reverse_jaw_pose)) -> (Output_Get:double)
   > Set jaw_up_pose(jaw_up_pose:double=GetCurveValue(CurveName=jaw_up_pose)) -> (Output_Get:double)
   > Set R_lower_lip_down_pose(R_lower_lip_down_pose:double=GetCurveValue(CurveName=R_lower_lip_down_pose)) -> (Output_Get:double)
   > Set R_upper_lip_raiser_pose(R_upper_lip_raiser_pose:double=GetCurveValue(CurveName=R_upper_lip_raiser_pose)) -> (Output_Get:double)
   > Set mouth_emote_Ee(mouth_emote_Ee:double=GetCurveValue(CurveName=mouth_emote_Ee)) -> (Output_Get:double)
   > Set mouth_emote_Oo(mouth_emote_Oo:double=GetCurveValue(CurveName=mouth_emote_Oo)) -> (Output_Get:double)
   > Set mouth_emote_01(mouth_emote_01:double=GetCurveValue(CurveName=mouth_emote_01)) -> (Output_Get:double)
   n2 = For Each Loop(
    Exec=@n1.then_1
    Array:Array of Skeletal Mesh Component Object References=Get Components By Class!(self:Actor Object Reference=GetOwningActor(), ComponentClass:Actor Component Class Reference=/Script/Engine.SkeletalMeshComponent) -> (ReturnValue:Array of Skeletal Mesh Component Object References)
   ) -> (Array Element:Skeletal Mesh Component Object Reference)
   n3 = Get R_brow_Ot_Dn_pose() -> (R_brow_Ot_Dn_pose:double)
   n4 = Get Threshold_Brow_Ot_Dn() -> (Threshold_Brow_Ot_Dn:double)
   n5 = Get R_brow_Ot_Dn_pose_Previous() -> (R_brow_Ot_Dn_pose_Previous:double)
   n6 = Get R_brow_In_Dn_pose() -> (R_brow_In_Dn_pose:double)
   n7 = Get Threshold_Brow_In_Dn() -> (Threshold_Brow_In_Dn:double)
   n8 = Get R_brow_In_Dn_pose_Previous() -> (R_brow_In_Dn_pose_Previous:double)
   n9 = AND Boolean(
    A="Not Equal (Boolean)"(A="float > float"(A=@n6, B=@n7), B="float > float"(A=@n8, B=@n7)) -> (ReturnValue=~)
    B="Not Equal (Boolean)"(A="float > float"(A=@n3, B=@n4), B="float > float"(A=@n5, B=@n4)) -> (ReturnValue=~)
   )
   n10 = Get R_brow_In_Up_pose() -> (R_brow_In_Up_pose:double)
   n11 = Get Threshold_Brow_In_Up() -> (Threshold_Brow_In_Up:double)
   n12 = Get R_brow_In_Up_pose_Previous() -> (R_brow_In_Up_pose_Previous:double)
   n13 = Get R_brow_Ot_Up_pose() -> (R_brow_Ot_Up_pose:double)
   n14 = Get Threshold_Brow_Ot_Up() -> (Threshold_Brow_Ot_Up:double)
   n15 = Get R_brow_Ot_Up_pose_Previous() -> (R_brow_Ot_Up_pose_Previous:double)
   n16 = AND Boolean(
    A="Not Equal (Boolean)"(A="float > float"(A=@n10, B=@n11), B="float > float"(A=@n12, B=@n11)) -> (ReturnValue=~)
    B="Not Equal (Boolean)"(A="float > float"(A=@n13, B=@n14), B="float > float"(A=@n15, B=@n14)) -> (ReturnValue=~)
   )
   n17 = Get R_brow_In_Dn_pose()
   n18 = Get Threshold_Brow_In_Dn()
   n19 = Get R_brow_In_Dn_pose_Previous()
   n20 = Get R_brow_squeeze_in_pose() -> (R_brow_squeeze_in_pose:double)
   n21 = Get Threshold_Brow_Squeeze_In() -> (Threshold_Brow_Squeeze_In:double)
   n22 = Get R_brow_squeeze_in_pose_Previous() -> (R_brow_squeeze_in_pose_Previous:double)
   n23 = AND Boolean(
    A="Not Equal (Boolean)"(A="float > float"(A=@n17, B=@n18), B="float > float"(A=@n19, B=@n18)) -> (ReturnValue=~)
    B="Not Equal (Boolean)"(A="float > float"(A=@n20, B=@n21), B="float > float"(A=@n22, B=@n21)) -> (ReturnValue=~)
   )
   n24 = Get R_brow_Ot_Dn_pose()
   n25 = Get Threshold_Brow_Ot_Dn()
   n26 = Get R_brow_Ot_Dn_pose_Previous()
   n27 = Get R_brow_squeeze_in_pose()
   n28 = Get Threshold_Brow_Squeeze_In()
   n29 = Get R_brow_squeeze_in_pose_Previous()
   n30 = AND Boolean(
    A="Not Equal (Boolean)"(A="float > float"(A=@n24, B=@n25), B="float > float"(A=@n26, B=@n25)) -> (ReturnValue=~)
    B="Not Equal (Boolean)"(A="float > float"(A=@n27, B=@n28), B="float > float"(A=@n29, B=@n28)) -> (ReturnValue=~)
   )
   n31 = Get Threshold_Blink() -> (Threshold_Blink:double)
   n32 = "float > float"(A=Get R_blink_pose() -> (R_blink_pose:double), B=@n31)
   n33 = "float > float"(A=Get R_blink_pose_Previous() -> (R_blink_pose_Previous:double), B=@n31)
   n34 = Get Threshold_Upper_Lid_Down() -> (Threshold_Upper_Lid_Down:double)
   n35 = "float > float"(A=Get R_upperLidDown_pose() -> (R_upperLidDown_pose:double), B=@n34)
   n36 = "float > float"(A=Get R_upperLidDown_pose_Previous() -> (R_upperLidDown_pose_Previous:double), B=@n34)
   n37 = Get R_smile_pose() -> (R_smile_pose:double)
   n38 = Get Threshold_Smile() -> (Threshold_Smile:double)
   n39 = Get R_smile_pose_Previous() -> (R_smile_pose_Previous:double)
   n40 = AND Boolean(
    A=OR Boolean(A="Not Equal (Boolean)"(A=@n32, B=@n33) -> (ReturnValue=~), B="Not Equal (Boolean)"(A=@n35, B=@n36) -> (ReturnValue=~))
    B="Not Equal (Boolean)"(A="float > float"(A=@n37, B=@n38), B="float > float"(A=@n39, B=@n38)) -> (ReturnValue=~)
   )
   n41 = Get R_blink_pose()
   n42 = Get Threshold_Blink()
   n43 = Get R_blink_pose_Previous()
   n44 = Get R_upperLidDown_pose()
   n45 = Get Threshold_Upper_Lid_Down()
   n46 = Get R_upperLidDown_pose_Previous()
   n47 = OR Boolean(
    A="Not Equal (Boolean)"(A="float > float"(A=@n41, B=@n42), B="float > float"(A=@n43, B=@n42)) -> (ReturnValue=~)
    B="Not Equal (Boolean)"(A="float > float"(A=@n44, B=@n45), B="float > float"(A=@n46, B=@n45)) -> (ReturnValue=~)
   )
   n48 = Get R_frown_pose() -> (R_frown_pose:double)
   n49 = Get Threshold_Frown() -> (Threshold_Frown:double)
   n50 = Get R_frown_pose_Previous() -> (R_frown_pose_Previous:double)
   n51 = Get jaw_open_pose() -> (jaw_open_pose:double)
   n52 = Get Threshold_Jaw_Open() -> (Threshold_Jaw_Open:double)
   n53 = Get jaw_open_pose_Previous() -> (jaw_open_pose_Previous:double)
   n54 = AND Boolean(
    A="Not Equal (Boolean)"(A="float > float"(A=@n48, B=@n49), B="float > float"(A=@n50, B=@n49)) -> (ReturnValue=~)
    B="Not Equal (Boolean)"(A="float > float"(A=@n51, B=@n52), B="float > float"(A=@n53, B=@n52)) -> (ReturnValue=~)
   )
   n55 = Get jaw_open_pose()
   n56 = Get Threshold_Jaw_Open()
   n57 = Get jaw_open_pose_Previous()
   n58 = Get reverse_jaw_pose() -> (reverse_jaw_pose:double)
   n59 = Get Threshold_Reverse_Jaw() -> (Threshold_Reverse_Jaw:double)
   n60 = Get reverse_jaw_pose_Previous() -> (reverse_jaw_pose_Previous:double)
   n61 = AND Boolean(
    A="Not Equal (Boolean)"(A="float > float"(A=@n55, B=@n56), B="float > float"(A=@n57, B=@n56)) -> (ReturnValue=~)
    B="Not Equal (Boolean)"(A="float > float"(A=@n58, B=@n59), B="float > float"(A=@n60, B=@n59)) -> (ReturnValue=~)
   )
   n62 = Get jaw_open_pose()
   n63 = Get Threshold_Jaw_Open()
   n64 = Get jaw_open_pose_Previous()
   n65 = Get R_smile_pose()
   n66 = Get Threshold_Smile()
   n67 = Get R_smile_pose_Previous()
   n68 = AND Boolean(
    A="Not Equal (Boolean)"(A="float > float"(A=@n62, B=@n63), B="float > float"(A=@n64, B=@n63)) -> (ReturnValue=~)
    B="Not Equal (Boolean)"(A="float > float"(A=@n65, B=@n66), B="float > float"(A=@n67, B=@n66)) -> (ReturnValue=~)
   )
   n69 = Get R_lower_lip_down_pose() -> (R_lower_lip_down_pose:double)
   n70 = Get Threshold_Lower_Lip_Down() -> (Threshold_Lower_Lip_Down:double)
   n71 = Get R_lower_lip_down_pose_Previous() -> (R_lower_lip_down_pose_Previous:double)
   n72 = Get R_upper_lip_raiser_pose() -> (R_upper_lip_raiser_pose:double)
   n73 = Get Threshold_Upper_Lip_Raiser() -> (Threshold_Upper_Lip_Raiser:double)
   n74 = Get R_upper_lip_raiser_pose_Previous() -> (R_upper_lip_raiser_pose_Previous:double)
   n75 = AND Boolean(
    A="Not Equal (Boolean)"(A="float > float"(A=@n69, B=@n70), B="float > float"(A=@n71, B=@n70)) -> (ReturnValue=~)
    B="Not Equal (Boolean)"(A="float > float"(A=@n72, B=@n73), B="float > float"(A=@n74, B=@n73)) -> (ReturnValue=~)
   )
   n76 = Get jaw_open_pose()
   n77 = Get Threshold_Jaw_Open()
   n78 = Get jaw_open_pose_Previous()
   n79 = Get R_lower_lip_down_pose()
   n80 = Get Threshold_Lower_Lip_Down()
   n81 = Get R_lower_lip_down_pose_Previous()
   n82 = AND Boolean(
    A="Not Equal (Boolean)"(A="float > float"(A=@n76, B=@n77), B="float > float"(A=@n78, B=@n77)) -> (ReturnValue=~)
    B="Not Equal (Boolean)"(A="float > float"(A=@n79, B=@n80), B="float > float"(A=@n81, B=@n80)) -> (ReturnValue=~)
   )
   n83 = Get Threshold_Brow_Emote_01() -> (Threshold_Brow_Emote_01:double)
   n84 = "Not Equal (Boolean)"(
    A="float > float"(A=Get brow_emote_01() -> (brow_emote_01:double), B=@n83)
    B="float > float"(A=Get brow_emote_01_Previous() -> (brow_emote_01_Previous:double), B=@n83)
   ) -> (ReturnValue=~)
   n85 = Get Threshold_Lower_Lid_Up() -> (Threshold_Lower_Lid_Up:double)
   n86 = "Not Equal (Boolean)"(
    A="float > float"(A=Get R_lower_lid_up_pose() -> (R_lower_lid_up_pose:double), B=@n85)
    B="float > float"(A=Get R_lower_lid_up_pose_Previous() -> (R_lower_lid_up_pose_Previous:double), B=@n85)
   ) -> (ReturnValue=~)
   n87 = Get Threshold_Eye_Emote_01() -> (Threshold_Eye_Emote_01:double)
   n88 = "Not Equal (Boolean)"(
    A="float > float"(A=Get eye_emote_01() -> (eye_emote_01:double), B=@n87)
    B="float > float"(A=Get eye_emote_01_Previous() -> (eye_emote_01_Previous:double), B=@n87)
   ) -> (ReturnValue=~)
   n89 = Get Threshold_Frown()
   n90 = "Not Equal (Boolean)"(A="float > float"(A=Get R_frown_pose(), B=@n89), B="float > float"(A=Get R_frown_pose_Previous(), B=@n89)) -> (ReturnValue=~)
   n91 = Get Threshold_Jaw_Open()
   n92 = "Not Equal (Boolean)"(A="float > float"(A=Get jaw_open_pose(), B=@n91), B="float > float"(A=Get jaw_open_pose_Previous(), B=@n91)) -> (ReturnValue=~)
   n93 = Get Threshold_Jaw_Up() -> (Threshold_Jaw_Up:double)
   n94 = "Not Equal (Boolean)"(
    A="float > float"(A=Get jaw_up_pose() -> (jaw_up_pose:double), B=@n93)
    B="float > float"(A=Get jaw_up_pose_Previous() -> (jaw_up_pose_Previous:double), B=@n93)
   ) -> (ReturnValue=~)
   n95 = Get Threshold_Smile()
   n96 = "Not Equal (Boolean)"(A="float > float"(A=Get R_smile_pose(), B=@n95), B="float > float"(A=Get R_smile_pose_Previous(), B=@n95)) -> (ReturnValue=~)
   n97 = Get Threshold_Mouth_Emote_Ee() -> (Threshold_Mouth_Emote_Ee:double)
   n98 = "Not Equal (Boolean)"(
    A="float > float"(A=Get mouth_emote_Ee() -> (mouth_emote_Ee:double), B=@n97)
    B="float > float"(A=Get mouth_emote_Ee_Previous() -> (mouth_emote_Ee_Previous:double), B=@n97)
   ) -> (ReturnValue=~)
   n99 = Get Threshold_Mouth_Emote_Oo() -> (Threshold_Mouth_Emote_Oo:double)
   n100 = "Not Equal (Boolean)"(
    A="float > float"(A=Get mouth_emote_Oo() -> (mouth_emote_Oo:double), B=@n99)
    B="float > float"(A=Get mouth_emote_Oo_Previous() -> (mouth_emote_Oo_Previous:double), B=@n99)
   ) -> (ReturnValue=~)
   n101 = Get Threshold_Mouth_Emote_01() -> (Threshold_Mouth_Emote_01:double)
   n102 = "Not Equal (Boolean)"(
    A="float > float"(A=Get mouth_emote_01() -> (mouth_emote_01:double), B=@n101)
    B="float > float"(A=Get mouth_emote_01_Previous() -> (mouth_emote_01_Previous:double), B=@n101)
   ) -> (ReturnValue=~)
   > SetScalarParameterValueOnMaterials!(
    execute
    self:Mesh Component Object Reference=@n2
    ParameterName:Name=HasPoseChanged
    ParameterValue:float="To Float (Boolean)"(InBool=OR Boolean!(
      A=@n9
      B=@n16
      C:bool=@n23
      D:bool=@n30
      E:bool=@n84
      F:bool=@n40
      G:bool=@n47
      H:bool=@n86
      I:bool=@n88
      J:bool=@n90
      K:bool=@n92
      L:bool=@n54
      M:bool=@n61
      N:bool=@n94
      O:bool=@n96
      P:bool=@n68
      Q:bool=@n75
      R:bool=@n82
      S:bool=@n98
      T:bool=@n100
      U:bool=@n102
     ) -> (ReturnValue))
   ) -> (then)
   Set R_brow_In_Dn_pose_Previous(execute=@n1.then_2, R_brow_In_Dn_pose_Previous:double=Get R_brow_In_Dn_pose()) -> (Output_Get:double)
   > Set R_brow_Ot_Dn_pose_Previous(R_brow_Ot_Dn_pose_Previous:double=Get R_brow_Ot_Dn_pose()) -> (Output_Get:double)
   > Set R_brow_In_Up_pose_Previous(R_brow_In_Up_pose_Previous:double=Get R_brow_In_Up_pose()) -> (Output_Get:double)
   > Set R_brow_Ot_Up_pose_Previous(R_brow_Ot_Up_pose_Previous:double=Get R_brow_Ot_Up_pose()) -> (Output_Get:double)
   > Set R_brow_squeeze_in_pose_Previous(R_brow_squeeze_in_pose_Previous:double=Get R_brow_squeeze_in_pose()) -> (Output_Get:double)
   > Set brow_emote_01_Previous(brow_emote_01_Previous:double=Get brow_emote_01()) -> (Output_Get:double)
   > Set R_blink_pose_Previous(R_blink_pose_Previous:double=Get R_blink_pose()) -> (Output_Get:double)
   > Set R_upperLidDown_pose_Previous(R_upperLidDown_pose_Previous:double=Get R_upperLidDown_pose()) -> (Output_Get:double)
   > Set R_smile_pose_Previous(R_smile_pose_Previous:double=Get R_smile_pose()) -> (Output_Get:double)
   > Set R_lower_lid_up_pose_Previous(R_lower_lid_up_pose_Previous:double=Get R_lower_lid_up_pose()) -> (Output_Get:double)
   > Set eye_emote_01_Previous(eye_emote_01_Previous:double=Get eye_emote_01()) -> (Output_Get:double)
   > Set R_frown_pose_Previous(R_frown_pose_Previous:double=Get R_frown_pose()) -> (Output_Get:double)
   > Set jaw_open_pose_Previous(jaw_open_pose_Previous:double=Get jaw_open_pose()) -> (Output_Get:double)
   > Set reverse_jaw_pose_Previous(reverse_jaw_pose_Previous:double=Get reverse_jaw_pose()) -> (Output_Get:double)
   > Set jaw_up_pose_Previous(jaw_up_pose_Previous:double=Get jaw_up_pose()) -> (Output_Get:double)
   > Set R_lower_lip_down_pose_Previous(R_lower_lip_down_pose_Previous:double=Get R_lower_lip_down_pose()) -> (Output_Get:double)
   > Set R_upper_lip_raiser_pose_Previous(R_upper_lip_raiser_pose_Previous:double=Get R_upper_lip_raiser_pose()) -> (Output_Get:double)
   > Set mouth_emote_Ee_Previous(mouth_emote_Ee_Previous:double=Get mouth_emote_Ee()) -> (Output_Get:double)
   > Set mouth_emote_Oo_Previous(mouth_emote_Oo_Previous:double=Get mouth_emote_Oo()) -> (Output_Get:double)
   > Set mouth_emote_01_Previous(mouth_emote_01_Previous:double=Get mouth_emote_01()) -> (Output_Get:double)
  ]}
 GetSequencerAnimSlotNames {nodes [
   GetSequencerAnimSlotNames() {is=title K2Node_FunctionEntry}
   > Return Node(ReturnValue:Array of Names)
  ]}
]
variables [
 brow_emote_01 {type=number}
 brow_emote_01_Previous {type=number}
 eye_emote_01 {type=number}
 eye_emote_01_Previous {type=number}
 jaw_open_pose {type=number}
 jaw_open_pose_Previous {type=number}
 jaw_up_pose {type=number}
 jaw_up_pose_Previous {type=number}
 mouth_emote_01 {type=number}
 mouth_emote_01_Previous {type=number}
 mouth_emote_Ee {type=number}
 mouth_emote_Ee_Previous {type=number}
 mouth_emote_Oo {type=number}
 mouth_emote_Oo_Previous {type=number}
 PoseChangeThreshold {type=number, default=0.01}
 R_blink_pose {type=number, default=0.4}
 R_blink_pose_Previous {type=number}
 R_brow_In_Dn_pose {type=number}
 R_brow_In_Dn_pose_Previous {type=number}
 R_brow_In_Dn_Thresholds {type=number}
 R_brow_In_Up_pose {type=number}
 R_brow_In_Up_pose_Previous {type=number}
 R_brow_Ot_Dn_pose {type=number}
 R_brow_Ot_Dn_pose_Previous {type=number}
 R_brow_Ot_Dn_Thresholds {type=number}
 R_brow_Ot_Up_pose {type=number}
 R_brow_Ot_Up_pose_Previous {type=number}
 R_brow_squeeze_in_pose {type=number}
 R_brow_squeeze_in_pose_Previous {type=number}
 R_frown_pose {type=number}
 R_frown_pose_Previous {type=number}
 R_lower_lid_up_pose {type=number}
 R_lower_lid_up_pose_Previous {type=number}
 R_lower_lip_down_pose {type=number}
 R_lower_lip_down_pose_Previous {type=number}
 R_smile_pose {type=number}
 R_smile_pose_Previous {type=number}
 R_upper_lip_raiser_pose {type=number}
 R_upper_lip_raiser_pose_Previous {type=number}
 R_upperLidDown_pose {type=number}
 R_upperLidDown_pose_Previous {type=number}
 reverse_jaw_pose {type=number}
 reverse_jaw_pose_Previous {type=number}
 Threshold_Blink {type=number, default=0.4}
 Threshold_Brow_Emote_01 {type=number}
 Threshold_Brow_In_Dn {type=number, default=0.1}
 Threshold_Brow_In_Up {type=number, default=0.1}
 Threshold_Brow_Ot_Dn {type=number, default=0.1}
 Threshold_Brow_Ot_Up {type=number, default=0.1}
 Threshold_Brow_Squeeze_In {type=number}
 Threshold_Eye_Emote_01 {type=number, default=0.4}
 Threshold_Frown {type=number, default=0.6}
 Threshold_Jaw_Open {type=number, default=0.25}
 Threshold_Jaw_Up {type=number, default=0.3}
 Threshold_Lower_Lid_Up {type=number, default=0.3}
 Threshold_Lower_Lip_Down {type=number, default=0.1}
 Threshold_Mouth_Emote_01 {type=number, default=0.1}
 Threshold_Mouth_Emote_Ee {type=number, default=0.1}
 Threshold_Mouth_Emote_Oo {type=number, default=0.1}
 Threshold_Reverse_Jaw {type=number, default=0.21}
 Threshold_Smile {type=number, default=0.4}
 Threshold_Upper_Lid_Down {type=number, default=0.4}
 Threshold_Upper_Lip_Raiser {type=number, default=0.1}
]
