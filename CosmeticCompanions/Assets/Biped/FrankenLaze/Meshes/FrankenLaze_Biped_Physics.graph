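# Physics asset for the FrankenLaze biped cosmetic companion mesh.
# It holds a single capsule body on the root bone, offset upward and slightly along Y, and rolled or tilted a quarter turn.
# There are no constraints, so the whole skeleton is represented by that one simple shape rather than a per-bone ragdoll.
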PhysicsAsset /CosmeticCompanions/Assets/Biped/FrankenLaze/Meshes/FrankenLaze_Biped_Physics
bodies {root [{shapeType=Capsule, length=26.505, radius=28.005, shapeName=root_capsule, center=<0 -7.9121 32.591>, rotation=<rot 0 -90 0>}]}
constraints []
